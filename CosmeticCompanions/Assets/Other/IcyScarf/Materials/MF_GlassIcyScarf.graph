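# Material function for the Icy Scarf cosmetic companion that builds frosted, translucent glass material attributes.
# Preskinned triplanar projection supplies frost mask and normal, keeping the pattern independent of skinned deformation.
# Frost and Fresnel terms shape base color, time-of-day-tinted emissive, metallic, roughness, flattened normal and opacity, with optional texture-driven opacity.
# Refraction is faked by offsetting Scene Color UVs, and a quality switch gives High and Epic a blended base color.
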MaterialFunction /CosmeticCompanions/Assets/Other/IcyScarf/Materials/MF_GlassIcyScarf
graph {nodes [
  Constant()
  FunctionOutput(None=NamedRerouteUsage())
  n1 = TextureObjectParameter() {
   group=IcyScarf
   parameterName=FrostTex
   samplerType=SAMPLERTYPE_Masks
   texture=/CosmeticCompanions/Assets/Other/IcyScarf/Textures/T_Frost_IcyScarf
  }
  n2 = ScalarParameter() {defaultValue=1.5, group=PreskinnedTriplanar, parameterName=UVTilingX, sortPriority=1}
  n3 = AppendVector(A=@n2, B=@n2)
  n4 = AppendVector(
   A=ScalarParameter() {defaultValue=0.124, group=PreskinnedTriplanar, parameterName=UVOffsetX, sortPriority=3}
   B=ScalarParameter() {defaultValue=0.212, group=PreskinnedTriplanar, parameterName=UVOffsetY, sortPriority=4}
  )
  n5 = ScalarParameter() {group=PreskinnedTriplanar, parameterName=BlendContrast, sortPriority=7}
  n6 = Power(
   Base=MaterialFunctionCall!(InputTexture=@n1, UV Tiling=@n3, UV Offset=@n4, Blend Contrast=@n5) {materialFunction=&/MF_PreskinnedTriplanar_IcyScarf}.Sample XYZ
   Exp=ScalarParameter() {defaultValue=1.5, group=IcyScarf, parameterName=FrostContrast}
  ) {constExponent=2.2}
  n7 = ComponentMask(None=@n6) {b=true}
  n8 = Constant() {r=0.8053}
  NamedRerouteDeclaration(None=Max(A=ComponentMask(None=@n6) {g=true}, B=Multiply(A=@n7, B=@n8))) {name=Frost, nodeColor=<rgba 1 0 0.3516 1>}
  n9 = ComponentMask(None=NamedRerouteUsage()) {r=true}
  n10 = LinearInterpolate(
   A=ScalarParameter() {defaultValue=0.15, group=IcyScarf, parameterName=InterBasecolorIntensity}
   Alpha=Power(Base=NamedRerouteUsage(), Exp=Constant() {r=2}) {constExponent=3}
  ) {constA=0.2}
  n11 = VectorParameter() {group=IcyScarf, parameterName=Frostcolor, defaultValue=<rgba 0.7826 0.854 0.9 1>}
  n12 = Fresnel(Normal=VertexNormalWS()) {exponent=1}
  n13 = LinearInterpolate(
   A=Constant() {r=1}
   B=Constant() {r=1.5}
   Alpha=Power(Base=NamedRerouteDeclaration(None=@n12) {name=FrostFresnel, nodeColor=<rgba 0.7812 1 0 1>}, Exp=Constant() {r=1})
  ) {constB=2}
  NamedRerouteDeclaration(None=Multiply(A=Multiply(A=@n9, B=@n10), B=Multiply(A=@n11.RGB, B=@n13))) {name=GlassBaseColor, nodeColor=<rgba 0 1 0.0625 1>}
  n14 = VertexNormalWS()
  n15 = Saturate(None=Power(Base=Fresnel(Normal=@n14) {exponent=2.04}, Exp=Constant() {r=1.8236}))
  n16 = LinearInterpolate(
   A=Constant() {r=0.25}
   B=MaterialFunctionCall() {materialFunction=/Game/Effects/Fort_Effects/Materials/_Functions/CurveColor_Time_of_Day}.Result
  ) {constAlpha=0.6}
  n17 = Power(Base=NamedRerouteUsage(), Exp=Constant() {r=2})
  n18 = Multiply(
   A=Multiply(
    A=ComponentMask(None=@n17) {b=true}
    B=VectorParameter() {group=IcyScarf, parameterName=frost color, defaultValue=<rgba 0.7121 0.9025 1 1>}.RGB
   )
   B=Power(Base=Fresnel() {exponent=1}, Exp=Constant() {r=1})
  )
  n19 = ScalarParameter() {defaultValue=0.3, group=IcyScarf, parameterName=Frost Emissive Intensity}
  NamedRerouteDeclaration(None=Add(A=Multiply(A=@n15, B=@n16) {constB=0.1}, B=Multiply(A=@n18, B=@n19) {constB=0.5})) {name=GlassEmissive, nodeColor=<rgba 0 0.9062 1 1>}
  NamedRerouteDeclaration(None=LinearInterpolate(Alpha=NamedRerouteUsage()) {constA=0.5}) {name=GlassSpecular, nodeColor=<rgba 0 1 0.0625 1>}
  NamedRerouteDeclaration(None=LinearInterpolate(
    A=ScalarParameter() {defaultValue=0.8255, group=IcyScarf, parameterName=GlassMetallicMin}
    B=ScalarParameter() {defaultValue=0.05, group=IcyScarf, parameterName=GlassMetallicMax}
    Alpha=Fresnel() {exponent=1}
   )) {name=GlassMetallic, nodeColor=<rgba 0 1 0.0625 1>}
  n20 = ComponentMask(None=NamedRerouteUsage()) {b=true}
  NamedRerouteDeclaration(None=LinearInterpolate(
    A=ScalarParameter() {defaultValue=0.02, group=IcyScarf, parameterName=GlassRoughnessMin}
    B=ScalarParameter() {defaultValue=0.2, group=IcyScarf, parameterName=GlassRoughnessMax}
    Alpha=Power(Base=@n20)
   )) {name=GlassRoughness, nodeColor=<rgba 0 1 0.0625 1>}
  n21 = NamedRerouteUsage()
  n22 = LinearInterpolate(A=@n21, B=NamedRerouteUsage(), Alpha=NamedRerouteUsage())
  n23 = MaterialFunctionCall!(
   Low _ PC Low and Med=@n21
   "High _ XB1, PS4"=@n22
   "Epic _ PC High, XSX, PS5"=@n22
   --------
   "Mobile Low _ PC Perf, Android"=@n21
   Mobile Med _ Switch=@n21
  ) {materialFunction=/Game/Global/MaterialFunctions/TechArt/MF_Fortnite_QualitySwitch_Vector3}
  n24 = Constant() {r=0.03}
  n25 = NamedRerouteUsage()
  n26 = SetMaterialAttributes!(
   MaterialAttributes
   BaseColor=Saturate(None=Add(A=@n23.Result, B=@n24))
   Opacity=@n25
   OpacityMask=@n25
   EmissiveColor=NamedRerouteUsage()
   Metallic=NamedRerouteUsage()
   Specular=NamedRerouteUsage()
   Roughness=NamedRerouteUsage()
   Normal=NamedRerouteUsage()
   Tangent=Constant3Vector() {constant=<rgba 1 0 0 0>}
  ) {attributeSetTypes [
    #1
    #2
    #3
    #4
    #5
    #6
    #7
    #8
    #9
   ]}
  NamedRerouteDeclaration(None=@n26) {name=GlassMaterial, nodeColor=<rgba 0.8516 0 1 1>}
  n27 = ScalarParameter() {defaultValue=1.2, group=IcyScarf, parameterName=InterNormalFlatness}
  n28 = ScalarParameter() {defaultValue=5, group=IcyScarf, parameterName=OuterNormalFlatness}
  n29 = Fresnel(Normal=VertexNormalWS()) {exponent=1}
  NamedRerouteDeclaration(None=MaterialFunctionCall(Normal=NamedRerouteUsage(), Flatness=LinearInterpolate(A=@n27, B=@n28, Alpha=@n29) {constA=1}) {materialFunction=&FlattenNormal}.Result) {name=GlassNormal, nodeColor=<rgba 0.8047 0 1 1>}
  n30 = Multiply(
   A=Fresnel(ExponentIn=ScalarParameter() {defaultValue=1, group=IcyScarf, parameterName=FresnelExponentTransparency, sortPriority=30})
   B=ScalarParameter() {defaultValue=1, group=IcyScarf, parameterName=GlassThickness, sortPriority=31}
  )
  n31 = LinearInterpolate(
   A=ScalarParameter() {group=IcyScarf, parameterName=InnerTransparency}
   B=ScalarParameter() {defaultValue=0.85, group=IcyScarf, parameterName=OuterTransparency, sortPriority=33}
   Alpha=NamedRerouteDeclaration(None=Saturate(None=@n30)) {name=GlassFresnel, nodeColor=<rgba 1 0.1875 0 1>}
  )
  NamedRerouteDeclaration(None=@n31) {name=GlassFresnelTransparency, nodeColor=<rgba 0 1 0.9531 1>}
  n32 = TextureObjectParameter() {
   group=IcyScarf
   parameterName=FrostN
   samplerType=SAMPLERTYPE_Normal
   texture=/CosmeticCompanions/Assets/Other/IcyScarf/Textures/T_Frost_IcyScarf_N
  }
  NamedRerouteDeclaration(None=MaterialFunctionCall(InputTexture=@n32, UV Tiling=@n3, UV Offset=@n4, Blend Contrast=@n5) {materialFunction=&MF_PreskinnedTriplanar_IcyScarf}.Sample XYZ) {name=FrostN, nodeColor=<rgba 0 0.6719 1 1>}
  NamedRerouteDeclaration(None=FunctionInput() {bUsePreviewValueAsDefault=true, inputName=TechArtMask, previewValue=<1 1 1 1>}) {name=TechArtMask, nodeColor=<rgba 0.8047 0 1 1>}
  NamedRerouteDeclaration(None=ChannelMaskParameter(None=NamedRerouteUsage()) {group=00_Glass, parameterName=TextureOpacityChannel, sortPriority=36}.R) {name=TextureOpactiy, nodeColor=<rgba 1 0.2578 0 1>}
  n33 = Multiply(A=NamedRerouteUsage(), B=@n31)
  n34 = NamedRerouteUsage()
  n35 = Multiply(A=@n34, B=ScalarParameter() {defaultValue=1, group=IcyScarf, parameterName=TextureOpacityAdd, sortPriority=34})
  n36 = NamedRerouteUsage()
  n37 = LinearInterpolate(A=Saturate(None=Add(A=@n35, B=@n36)), B=@n34, Alpha=ScalarParameter() {group=IcyScarf, parameterName=TextureOpacityBlend, sortPriority=35})
  n38 = NamedRerouteUsage()
  NamedRerouteDeclaration(None=Add(A=Saturate(None=@n33), B=StaticSwitchParameter(True=@n37, False=@n38) {group=IcyScarf, parameterName=UseTextureOpacity, sortPriority=36})) {name=GlassOpacity, nodeColor=<rgba 0.9922 0 1 1>}
  n39 = Fresnel(Normal=PixelNormalWS()) {exponent=1.2}
  n40 = Saturate(None=Power(Base=Multiply(A=@n39) {constB=1.2}) {constExponent=4})
  n41 = Multiply(A=@n39, B=ScalarParameter() {defaultValue=4.3204, group=IcyScarf, parameterName=FakeRefractionTimes})
  n42 = LinearInterpolate(Alpha=Saturate(None=Sine(None=@n41))) {constA=0.3, constB=0.6}
  n43 = ScalarParameter() {defaultValue=0.2, group=IcyScarf, parameterName=FakeRefractionMIn}
  n44 = ScalarParameter() {defaultValue=1.5, parameterName=FakeRefractionMax}
  n45 = ComponentMask(None=NamedRerouteUsage()) {r=true}
  n46 = Add(A=ScreenPosition().ViewportUV, B=Multiply(A=Multiply(A=@n40, B=@n42), B=LinearInterpolate(A=@n43, B=@n44, Alpha=@n45)))
  NamedRerouteDeclaration(None=Multiply(A=SceneColor!(UVs=@n46).RGB, B=ScalarParameter() {defaultValue=0.7, group=IcyScarf, parameterName=FakeRefraction}) {constB=0.9}) {name=FakeRefraction, nodeColor=<rgba 1 0 0.4687 1>}
 ]}
