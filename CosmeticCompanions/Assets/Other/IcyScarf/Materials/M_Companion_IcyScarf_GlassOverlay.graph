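# Additive, two-sided lit glass overlay material that adds frost to the IcyScarf cosmetic companion.
# Pre-skinned triplanar functions project the frost mask and normal map, keeping them fixed to the deforming mesh.
# The contrast-adjusted mask drives opacity, limited to back faces by TwoSidedSign, and base color, tinted by a time-of-day curve.
# These override attributes from the character PreFX and companion custom-color functions before passing into PostFX.
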Material /CosmeticCompanions/Assets/Other/IcyScarf/Materials/M_Companion_IcyScarf_GlassOverlay
graph {
 properties {
  bAutomaticallySetUsageInEditor=false
  bScreenSpaceReflections=true
  bUseMaterialAttributes=true
  bUsedWithMeshDeformer=true
  bUsedWithMorphTargets=true
  bUsedWithSkeletalMesh=true
  bUsedWithStaticLighting=true
  blendMode=BLEND_Additive
  translucencyLightingMode=TLM_Surface
  twoSided=true
  usedShadingModels=Default Lit
 }
 nodes [
  n1 = MaterialFunctionCall() {materialFunction=&MF_CharacterParent_2023_PreFX}
  n2 = MaterialFunctionCall!(Material=@n1.Result, Tech Art Mask=@n1.Result) {materialFunction=/CosmeticCompanions/Materials/MaterialFunctions/MF_FN_Companion_Custom_Color}
  n3 = SetMaterialAttributes!(
   MaterialAttributes=GetMaterialAttributes(None=NamedRerouteDeclaration(None=@n2.Base Color) {name=InputMA, nodeColor=<rgba 0 1 0.6719 1>}) {attributeGetTypes [#1]}.MaterialAttributes
   Opacity=NamedRerouteUsage()
   Normal=NamedRerouteUsage()
   Base Color=NamedRerouteUsage()
   Metallic=Constant() {r=0.15}
   Roughness=Constant() {r=0.304}
  ) {attributeSetTypes [
    #2
    #3
    #1
    #4
    #5
   ]}
  MaterialFunctionCall(Material=@n3, M=@n1.M, Pre Skinned Local Position=@n1.M, Distance Blend=@n1.M, Fresnel=@n1.M) {materialFunction=&MF_CharacterParent_2023_PostFX}
  n4 = ScalarParameter() {defaultValue=1.5, group=PreskinnedTriplanar, parameterName=UVTilingX, sortPriority=1}
  n5 = AppendVector(A=@n4, B=@n4)
  n6 = AppendVector(
   A=ScalarParameter() {defaultValue=0.124, group=PreskinnedTriplanar, parameterName=UVOffsetX, sortPriority=3}
   B=ScalarParameter() {defaultValue=0.212, group=PreskinnedTriplanar, parameterName=UVOffsetY, sortPriority=4}
  )
  n7 = ScalarParameter() {group=PreskinnedTriplanar, parameterName=BlendContrast, sortPriority=7}
  NamedRerouteDeclaration(None=MaterialFunctionCall!(
    InputTexture=TextureObject() {samplerType=SAMPLERTYPE_Masks, texture=/CosmeticCompanions/Assets/Other/IcyScarf/Textures/T_Frost_IcyScarf}
    UV Tiling=@n5
    UV Offset=@n6
    Blend Contrast=@n7
   ) {materialFunction=&/MF_PreskinnedTriplanar_IcyScarf}.Sample XYZ) {name=Frost, nodeColor=<rgba 1 0 0.3516 1>}
  n8 = Power(
   Base=ComponentMask(None=NamedRerouteUsage()) {g=true}
   Exp=ScalarParameter() {defaultValue=1.6992, group=IcyScarf, parameterName=FrostContrast}
  ) {constExponent=4}
  n9 = Max(A=Multiply(A=@n8, B=Constant() {r=0.8}) {constB=0.5}, B=Constant() {r=0.3})
  n10 = OneMinus(None=Saturate(None=TwoSidedSign()))
  NamedRerouteDeclaration(None=Multiply(A=Multiply(A=@n9, B=@n10), B=ScalarParameter() {defaultValue=0.586, group=IcyScarf, parameterName=FrostOpacity}) {constB=0.75}) {name=Opacity, nodeColor=<rgba 1 0 0.4453 1>}
  n11 = Constant() {r=0.8}
  n12 = MaterialFunctionCall() {materialFunction=/Game/Effects/Fort_Effects/Materials/_Functions/CurveColor_Time_of_Day}
  NamedRerouteDeclaration(None=Multiply(A=Add(A=@n8, B=@n11), B=Multiply(A=@n12.Result) {constB=0.35})) {name=BaseColor, nodeColor=<rgba 0 0.7656 1 1>}
  NamedRerouteDeclaration(None=MaterialFunctionCall(
    InputTexture=TextureObject() {samplerType=SAMPLERTYPE_Normal, texture=/CosmeticCompanions/Assets/Other/IcyScarf/Textures/T_Frost_IcyScarf_N}
    UV Tiling=@n5
    UV Offset=@n6
    Blend Contrast=@n7
   ) {materialFunction=&MF_PreskinnedTriplanar_IcyScarf}.Sample XYZ) {name=FrostN, nodeColor=<rgba 0 0.6719 1 1>}
 ]
}
