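# Post-process style Animation Blueprint for the IcyScarf "Other" cosmetic skeleton, adding secondary motion and pose-asset blending.
# The input pose runs through an Anim Dynamics node that simulates the dyn_hat bone in root-relative space, with gravity, an angular spring, damping and angular limits.
# The result feeds a linear Pose Blend node using the PoseAsset variable, defaulting to IcyScarf_Other_PoseAsset, before reaching the output pose.
# The event graph is empty.
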AnimBlueprint /CosmeticCompanions/Assets/Other/IcyScarf/Meshes/IcyScarf_Other_Post_AnimBP
target_skeleton=&/IcyScarf_Other_Skeleton
graphs [
 AnimGraph {nodes [
   n1 = Anim Dynamics - Bone: dyn_hat(ComponentPose=Local To Component(LocalPose=Input Pose()), Alpha=1, ExternalForce="0,0,0") {
    angularDampingOverride=0.9999
    angularSpringConstant=3000
    bAngularSpring=true
    bOverrideAngularDamping=true
    bUseGravityOverride=true
    simulationSpace=RootRelative
    boundBone {boneName=dyn_hat}
    componentAppliedLinearAccClamp=<4000 4000 4000>
    gravityOverride {z=-980}
    physicsBodyDefinitions [{
      boundBone {boneName=dyn_hat}
      boxExtents=<16 16 16>
      constraintSetup {angularLimitsMax=<10 0 90>, angularLimitsMin=<-10 -45 -90>, angularTarget=<0 0 0>}
      localJointOffset=<3 0 0>
     }]
    simSpaceSettings {maxAngularAcceleration=50, maxAngularVelocity=10, simSpaceAngularAlpha=1}
   }
   Output Pose(Result=Pose!(
     PoseAsset:Pose Asset Object Reference=Get PoseAsset() -> (PoseAsset:Pose Asset Object Reference)
     SourcePose:Pose Link Structure=Component To Local(ComponentPose=@n1)
    ) -> (Pose:Pose Link Structure) {is=title AnimGraphNode_PoseBlendNode, blendOption=Linear})
  ]}
 EventGraph {links [], nodes []}
]
variables [PoseAsset {type=PoseAsset, default=&/IcyScarf_Other_PoseAsset}]
